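# Physics asset for the female medium Scholar that, despite its Backpack name, defines dynamic hair and bangs bodies.
# Base spheres root short bangs chains and longer front, side and back hair chains of mostly boxes, with tight swing and twist limits.
# Angular-free constraints cross-link neighboring front, side and back strands, tying them together laterally.
# Head, neck, upper spine, clavicle and upper-arm capsules form unconstrained body shapes, likely as hair colliders.
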PhysicsAsset /Game/Characters/Player/Female/Medium/Bodies/F_MED_Scholar/Meshes/Parts/F_MED_Scholar_Backpack_Physics
bodies {
 clavicle_l [{shapeType=Capsule, length=13.7479, radius=6.7759, center=<0.0001 9.908 -9.5475>, rotation=<rot 0 0 127.4609>}]
 clavicle_r [{shapeType=Capsule, length=12.584, radius=7.9398, center=<-4.7355 -9.0676 8.9147>, rotation=<rot 0 0 131.0607>}]
 dyn_bangs_01_l [{shapeType=Box, extentX=1.8315, extentY=2.6465, extentZ=0.5, center=<0.7494 0.0001 0>}]
 dyn_bangs_01_mid [{shapeType=Box, extentX=1.8182, extentY=3.0093, extentZ=1.2773, center=<0.9021 0 0>}]
 dyn_bangs_01_r [{shapeType=Box, extentX=1.26, extentY=2.51, extentZ=0.76, center=<0.8779 0 0>}]
 dyn_bangs_base_l [{shapeType=Sphere, radius=0.5}]
 dyn_bangs_base_mid [{shapeType=Sphere, radius=0.5}]
 dyn_bangs_base_r [{shapeType=Sphere, radius=0.5}]
 dyn_bangs_parent_l [{shapeType=Box, extentX=2.01, extentY=2.51, extentZ=1.01, center=<0.9769 0 0>}]
 dyn_bangs_parent_mid [{shapeType=Box, extentX=1.8274, extentY=3.0958, extentZ=1.0803, center=<0.9652 0 0>}]
 dyn_bangs_parent_r [{shapeType=Box, extentX=1.9413, extentY=2.5657, extentZ=0.6681, center=<1.0021 0 0>}]
 dyn_hair_back_01_mid [{shapeType=Box, extentX=7.5, extentY=7.5, extentZ=2.2217, center=<3.6183 0 -0.0002>}]
 dyn_hair_back_02_mid [{shapeType=Box, extentX=7.5, extentY=7.5, extentZ=2.2217, center=<3.6183 0 -2>}]
 dyn_hair_back_03_mid [{shapeType=Box, extentX=7.5, extentY=7.5, extentZ=2.2217, center=<3.6183 0 -2>}]
 dyn_hair_back_04_mid [{shapeType=Box, extentX=7.5, extentY=7.5, extentZ=4.7217, center=<3.6183 0 -2>}]
 dyn_hair_back_05_mid [{shapeType=Box, extentX=9.0881, extentY=7.5, extentZ=9.4667, center=<4.431 0 -1.2197>}]
 dyn_hair_back_base_mid [{shapeType=Sphere, radius=0.5}]
 dyn_hair_back_parent_mid [{shapeType=Box, extentX=7.5, extentY=7.5, extentZ=2.2217, center=<3.6183 0 -0.0002>}]
 dyn_hair_front_01_l [{shapeType=Box, extentX=7.5, extentY=7.5, extentZ=2.2217, center=<3.6183 0 -0.0002>}]
 dyn_hair_front_01_r [{shapeType=Box, extentX=5.1, extentY=7.5, extentZ=2.2217, center=<2.8016 0 -0.0002>}]
 dyn_hair_front_02_l [{shapeType=Box, extentX=7.5, extentY=7.5, extentZ=3.2217, center=<2 -0.0001 -3>, rotation=<rot -54.4066 0 0>}]
 dyn_hair_front_02_r [{shapeType=Capsule, length=4.3786, radius=1.5, center=<2 -1 0>, rotation=<rot 0 0 -97.7838>}]
 dyn_hair_front_03_l [{shapeType=Box, extentX=7.5, extentY=7.5, extentZ=2.2217, center=<3.6183 0 -0.0002>}]
 dyn_hair_front_03_r [{shapeType=Capsule, length=6.728, radius=2, center=<0 -4.742 -1.1855>, rotation=<rot 7.8555 39.349 -99.4639>}]
 dyn_hair_front_04_l [{shapeType=Box, extentX=9.1962, extentY=7.5, extentZ=3.6875, center=<4.3864 0 1.2488>, rotation=<rot 29.2309 0 0>}]
 dyn_hair_front_base_l [{shapeType=Sphere, radius=0.5}]
 dyn_hair_front_base_r [{shapeType=Sphere, radius=0.5}]
 dyn_hair_front_parent_l [{shapeType=Box, extentX=7.5, extentY=7.5, extentZ=2.2217, center=<3.6183 0 -0.0002>}]
 dyn_hair_front_parent_r [{shapeType=Box, extentX=6.2798, extentY=7.5, extentZ=2.2217, center=<3.2855 0 -0.0002>}]
 dyn_hair_side_01_l [{shapeType=Box, extentX=7.5, extentY=7.5, extentZ=2.2217, center=<3.6183 0 -0.0002>}]
 dyn_hair_side_01_r [{shapeType=Box, extentX=7.5, extentY=7.5, extentZ=2.2217, center=<3.6183 0 -2>, rotation=<rot 0 0 33.9804>}]
 dyn_hair_side_02_l [{shapeType=Box, extentX=7.5, extentY=7.5, extentZ=2.2217, center=<3.6183 0 -0.0002>}]
 dyn_hair_side_02_r [{shapeType=Box, extentX=7.5, extentY=5.5, extentZ=6.5533, center=<3.6 0 -2>, rotation=<rot 0 0 33.9804>}]
 dyn_hair_side_03_l [{shapeType=Box, extentX=7.5, extentY=7.5, extentZ=2.2217, center=<3.6183 0 -0.0002>}]
 dyn_hair_side_03_r [{shapeType=Box, extentX=7.5, extentY=5.5, extentZ=5.5533, center=<3.6 0 -3>, rotation=<rot 0 0 15>}]
 dyn_hair_side_04_l [{shapeType=Box, extentX=7.5, extentY=7.5, extentZ=2.2217, center=<3.6183 0 -0.0002>}]
 dyn_hair_side_04_r [{shapeType=Box, extentX=9.1977, extentY=7.5, extentZ=7.2217, center=<4 -0.7254 -2>, rotation=<rot 0 0 33.9807>}]
 dyn_hair_side_05_l [{shapeType=Box, extentX=8.5209, extentY=7.5, extentZ=3.3231, center=<2.7586 0 -2>}]
 dyn_hair_side_base_l [{shapeType=Sphere, radius=0.5}]
 dyn_hair_side_base_r [{shapeType=Sphere, radius=0.5}]
 dyn_hair_side_parent_l [{shapeType=Box, extentX=7.5, extentY=7.5, extentZ=2.2217, center=<3.6183 0 -0.0002>}]
 dyn_hair_side_parent_r [{shapeType=Box, extentX=7.5, extentY=7.5, extentZ=2.2217, center=<3.6183 0 -0.0002>, rotation=<rot 0 0 12.533>}]
 head [{shapeType=Capsule, length=3.625, radius=6, center=<2.4384 2.1993 -0.0001>, rotation=<rot 90 0 -9.2448>}]
 neck_02 [{shapeType=Capsule, length=4.2601, radius=6.7399, center=<-2.6691 0.5785 0>, rotation=<rot 90 0 0>}]
 spine_05 [
  {shapeType=Capsule, length=8, radius=11.25, center=<0.7916 1.0898 0>, rotation=<rot 90 165.9638 155.6705>}
  {shapeType=Capsule, length=22.2782, radius=7.7218, center=<6.3018 -0.8102 0.0001>}
 ]
 upperarm_l [{shapeType=Capsule, length=25.1067, radius=5.5234, center=<11.5868 0.0001 0>, rotation=<rot 90 0 0>}]
 upperarm_r [{shapeType=Capsule, length=21.8556, radius=5.7413, center=<-10.7816 -0.0001 0>, rotation=<rot 90 0 0>}]
}
constraints [
 {bone1Name=dyn_hair_front_parent_r, bone2Name=dyn_hair_front_base_r, swing1LimitDegrees=30, swing2LimitDegrees=15, twistLimitDegrees=5}
 {bone1Name=dyn_hair_front_01_r, bone2Name=dyn_hair_front_parent_r, swing1LimitDegrees=30, swing2LimitDegrees=15, twistLimitDegrees=5}
 {bone1Name=dyn_hair_front_02_r, bone2Name=dyn_hair_front_01_r, swing1LimitDegrees=30, swing2LimitDegrees=15, twistLimitDegrees=5}
 {bone1Name=dyn_hair_front_03_r, bone2Name=dyn_hair_front_02_r, swing1LimitDegrees=30, swing2LimitDegrees=15, twistLimitDegrees=5}
 {bone1Name=dyn_hair_side_parent_r, bone2Name=dyn_hair_side_base_r, swing1LimitDegrees=8, swing2LimitDegrees=30, twistLimitDegrees=5}
 {bone1Name=dyn_hair_side_01_r, bone2Name=dyn_hair_side_parent_r, swing1LimitDegrees=8, swing2LimitDegrees=30, twistLimitDegrees=5}
 {bone1Name=dyn_hair_side_02_r, bone2Name=dyn_hair_side_01_r, swing1LimitDegrees=8, swing2LimitDegrees=30, twistLimitDegrees=5}
 {bone1Name=dyn_hair_side_03_r, bone2Name=dyn_hair_side_02_r, swing1LimitDegrees=10, swing2LimitDegrees=20, twistLimitDegrees=5}
 {bone1Name=dyn_hair_side_04_r, bone2Name=dyn_hair_side_03_r, swing1LimitDegrees=8, swing2LimitDegrees=10, twistLimitDegrees=5}
 {bone1Name=dyn_hair_back_parent_mid, bone2Name=dyn_hair_back_base_mid, swing1LimitDegrees=5, swing2LimitDegrees=30, twistLimitDegrees=5}
 {bone1Name=dyn_hair_back_01_mid, bone2Name=dyn_hair_back_parent_mid, swing1LimitDegrees=5, swing2LimitDegrees=30, twistLimitDegrees=5}
 {bone1Name=dyn_hair_back_02_mid, bone2Name=dyn_hair_back_01_mid, swing1LimitDegrees=5, swing2LimitDegrees=30, twistLimitDegrees=5}
 {bone1Name=dyn_hair_back_03_mid, bone2Name=dyn_hair_back_02_mid, swing1LimitDegrees=5, swing2LimitDegrees=30, twistLimitDegrees=5}
 {bone1Name=dyn_hair_back_04_mid, bone2Name=dyn_hair_back_03_mid, swing1LimitDegrees=5, swing2LimitDegrees=20, twistLimitDegrees=5}
 {bone1Name=dyn_hair_back_05_mid, bone2Name=dyn_hair_back_04_mid, swing1LimitDegrees=5, swing2LimitDegrees=10, twistLimitDegrees=5}
 {bone1Name=dyn_hair_side_parent_l, bone2Name=dyn_hair_side_base_l, swing1LimitDegrees=5, swing2LimitDegrees=30, twistLimitDegrees=5}
 {bone1Name=dyn_hair_side_01_l, bone2Name=dyn_hair_side_parent_l, swing1LimitDegrees=5, swing2LimitDegrees=30, twistLimitDegrees=5}
 {bone1Name=dyn_hair_side_02_l, bone2Name=dyn_hair_side_01_l, swing1LimitDegrees=5, swing2LimitDegrees=30, twistLimitDegrees=5}
 {bone1Name=dyn_hair_side_03_l, bone2Name=dyn_hair_side_02_l, swing1LimitDegrees=5, swing2LimitDegrees=20, twistLimitDegrees=5}
 {bone1Name=dyn_hair_side_04_l, bone2Name=dyn_hair_side_03_l, swing1LimitDegrees=5, swing2LimitDegrees=20, twistLimitDegrees=5}
 {bone1Name=dyn_hair_side_05_l, bone2Name=dyn_hair_side_04_l, swing1LimitDegrees=5, swing2LimitDegrees=10, twistLimitDegrees=5}
 {bone1Name=dyn_hair_front_parent_l, bone2Name=dyn_hair_front_base_l, swing1LimitDegrees=8, swing2LimitDegrees=20, twistLimitDegrees=5}
 {bone1Name=dyn_hair_front_01_l, bone2Name=dyn_hair_front_parent_l, swing1LimitDegrees=8, swing2LimitDegrees=20, twistLimitDegrees=5}
 {bone1Name=dyn_hair_front_02_l, bone2Name=dyn_hair_front_01_l, swing1LimitDegrees=5, swing2LimitDegrees=20, twistLimitDegrees=8}
 {bone1Name=dyn_hair_front_03_l, bone2Name=dyn_hair_front_02_l, swing1LimitDegrees=5, swing2LimitDegrees=10, twistLimitDegrees=8}
 {bone1Name=dyn_hair_front_04_l, bone2Name=dyn_hair_front_03_l, swing1LimitDegrees=5, swing2LimitDegrees=10, twistLimitDegrees=8}
 {bone1Name=dyn_bangs_parent_r, bone2Name=dyn_bangs_base_r, swing1LimitDegrees=10, swing2LimitDegrees=4, twistLimitDegrees=3}
 {bone1Name=dyn_bangs_01_r, bone2Name=dyn_bangs_parent_r, swing1LimitDegrees=10, swing2LimitDegrees=8, twistLimitDegrees=3}
 {bone1Name=dyn_bangs_parent_mid, bone2Name=dyn_bangs_base_mid, swing1LimitDegrees=10, swing2LimitDegrees=4, twistLimitDegrees=3}
 {bone1Name=dyn_bangs_01_mid, bone2Name=dyn_bangs_parent_mid, swing1LimitDegrees=10, swing2LimitDegrees=8, twistLimitDegrees=3}
 {bone1Name=dyn_bangs_parent_l, bone2Name=dyn_bangs_base_l, swing1LimitDegrees=10, swing2LimitDegrees=4, twistLimitDegrees=3}
 {bone1Name=dyn_bangs_01_l, bone2Name=dyn_bangs_parent_l, swing1LimitDegrees=10, swing2LimitDegrees=8, twistLimitDegrees=3}
 {
  bone1Name=dyn_hair_side_04_l
  bone2Name=dyn_hair_back_04_mid
  swing1LimitDegrees=45
  swing1Motion=Free
  swing2LimitDegrees=45
  swing2Motion=Free
  twistLimitDegrees=45
  twistMotion=Free
 }
 {
  bone1Name=dyn_hair_side_03_r
  bone2Name=dyn_hair_back_03_mid
  swing1LimitDegrees=45
  swing1Motion=Free
  swing2LimitDegrees=45
  swing2Motion=Free
  twistLimitDegrees=45
  twistMotion=Free
 }
 {
  bone1Name=dyn_hair_front_03_l
  bone2Name=dyn_hair_side_03_l
  swing1LimitDegrees=45
  swing1Motion=Free
  swing2LimitDegrees=45
  swing2Motion=Free
  twistLimitDegrees=45
  twistMotion=Free
 }
 {
  bone1Name=dyn_hair_front_04_l
  bone2Name=dyn_hair_side_04_l
  swing1LimitDegrees=45
  swing1Motion=Free
  swing2LimitDegrees=45
  swing2Motion=Free
  twistLimitDegrees=45
  twistMotion=Free
 }
 {
  bone1Name=dyn_hair_side_04_r
  bone2Name=dyn_hair_back_05_mid
  swing1LimitDegrees=45
  swing1Motion=Free
  swing2LimitDegrees=45
  swing2Motion=Free
  twistLimitDegrees=45
  twistMotion=Free
 }
 {
  bone1Name=dyn_hair_front_02_l
  bone2Name=dyn_hair_side_02_l
  swing1LimitDegrees=45
  swing1Motion=Free
  swing2LimitDegrees=45
  swing2Motion=Free
  twistLimitDegrees=45
  twistMotion=Free
 }
 {
  bone1Name=dyn_hair_side_02_l
  bone2Name=dyn_hair_back_02_mid
  swing1LimitDegrees=45
  swing1Motion=Free
  swing2LimitDegrees=45
  swing2Motion=Free
  twistLimitDegrees=45
  twistMotion=Free
 }
 {
  bone1Name=dyn_hair_side_03_l
  bone2Name=dyn_hair_back_03_mid
  swing1LimitDegrees=45
  swing1Motion=Free
  swing2LimitDegrees=45
  swing2Motion=Free
  twistLimitDegrees=45
  twistMotion=Free
 }
 {
  bone1Name=dyn_hair_side_05_l
  bone2Name=dyn_hair_back_05_mid
  swing1LimitDegrees=45
  swing1Motion=Free
  swing2LimitDegrees=45
  swing2Motion=Free
  twistLimitDegrees=45
  twistMotion=Free
 }
]
