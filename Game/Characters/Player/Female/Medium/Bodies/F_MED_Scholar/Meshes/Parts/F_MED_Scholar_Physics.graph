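# Physics asset for the medium female Scholar player body, set up for dynamic hair and bangs.
# Bangs, front, side and back bone chains use mostly box bodies, each chained from a small base sphere with tight swing and twist limits.
# Cross constraints with wide or free angular motion tie neighboring front, side and back chains together, more densely on the left.
# Capsules on head, neck, upper spine, clavicles and upper arms form the body shapes around the hair.
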PhysicsAsset /Game/Characters/Player/Female/Medium/Bodies/F_MED_Scholar/Meshes/Parts/F_MED_Scholar_Physics
bodies {
 clavicle_l [{shapeType=Capsule, length=14.7479, radius=5.77, center=<4.486 0.0001 4.0745>, rotation=<rot 90 0 0>}]
 clavicle_r [{shapeType=Capsule, length=14.7479, radius=7, center=<-4.4859 0.0001 -1.3907>, rotation=<rot 90 0 0>}]
 dyn_bangs_01_l [{shapeType=Box, extentX=1.8315, extentY=2.6465, extentZ=0.5, center=<0.7494 0.0001 0>}]
 dyn_bangs_01_mid [{shapeType=Box, extentX=1.8182, extentY=3.0093, extentZ=1.2773, center=<0.9021 0 0>}]
 dyn_bangs_01_r [{shapeType=Box, extentX=1.26, extentY=2.51, extentZ=0.76, center=<0.8779 0 0>}]
 dyn_bangs_base_l [{shapeType=Sphere, radius=0.5}]
 dyn_bangs_base_mid [{shapeType=Sphere, radius=0.5}]
 dyn_bangs_base_r [{shapeType=Sphere, radius=0.5}]
 dyn_bangs_parent_l [{shapeType=Box, extentX=2.01, extentY=2.51, extentZ=1.01, center=<0.9769 0 0>}]
 dyn_bangs_parent_mid [{shapeType=Box, extentX=1.8274, extentY=3.0958, extentZ=1.0803, center=<0.9652 0 0>}]
 dyn_bangs_parent_r [{shapeType=Box, extentX=1.9413, extentY=2.5657, extentZ=0.6681, center=<1.0021 0 0>}]
 dyn_hair_back_01_mid [{shapeType=Box, extentX=7.5, extentY=7.5, extentZ=2.2217, center=<3.6183 0 -1>}]
 dyn_hair_back_02_mid [{shapeType=Box, extentX=7.5, extentY=7.5, extentZ=2.2217, center=<3.6183 0 -1>}]
 dyn_hair_back_03_mid [{shapeType=Box, extentX=7.5, extentY=7.5, extentZ=2.2217, center=<3.6183 0 -1>}]
 dyn_hair_back_04_mid [{shapeType=Box, extentX=7.5, extentY=7.5, extentZ=3.7217, center=<3.6183 0 -1>}]
 dyn_hair_back_05_mid [{shapeType=Box, extentX=9.0881, extentY=9, extentZ=3.4667, center=<4.431 0 -1>}]
 dyn_hair_back_base_mid [{shapeType=Sphere, radius=0.5}]
 dyn_hair_back_parent_mid [{shapeType=Box, extentX=7.5, extentY=7.5, extentZ=2.2217, center=<3.6183 0 -0.0002>}]
 dyn_hair_front_01_l [{shapeType=Box, extentX=7.5, extentY=7.5, extentZ=2.2217, center=<3.6183 0 -0.0002>, rotation=<rot 0 0 35.0001>}]
 dyn_hair_front_01_r [{shapeType=Box, extentX=5.1, extentY=7.5, extentZ=2.2217, center=<2.8016 -0.8231 -0.0002>}]
 dyn_hair_front_02_l [{shapeType=Box, extentX=7.5, extentY=7.5, extentZ=2.2217, center=<2.9365 -0.0001 -2.5173>, rotation=<rot -54.4062 0.0001 24.9986>}]
 dyn_hair_front_02_r [{shapeType=Capsule, length=4.8786, radius=1.5, center=<2 -1 0>, rotation=<rot 0 0 -97.7838>}]
 dyn_hair_front_03_l [{shapeType=Box, extentX=7.5, extentY=7.5, extentZ=3.2217, center=<3.6183 0 -0.0002>}]
 dyn_hair_front_03_r [{shapeType=Capsule, length=5.478, radius=1.5, center=<0 -5 0>, rotation=<rot 7.8555 39.349 -99.4639>}]
 dyn_hair_front_04_l [{shapeType=Box, extentX=9.1962, extentY=7.5, extentZ=4.1875, center=<5.0111 0 -0.3326>, rotation=<rot 29.2309 0 0>}]
 dyn_hair_front_base_l [{shapeType=Sphere, radius=0.5}]
 dyn_hair_front_base_r [{shapeType=Sphere, radius=0.5}]
 dyn_hair_front_parent_l [{shapeType=Box, extentX=7.5, extentY=7.5, extentZ=2.2217, center=<3.6183 0 -0.0002>}]
 dyn_hair_front_parent_r [{shapeType=Box, extentX=6.2798, extentY=7.5, extentZ=2.2217, center=<3.2855 0 -0.0002>}]
 dyn_hair_side_01_l [{shapeType=Box, extentX=7.5, extentY=7.5, extentZ=2.2217, center=<3.6183 0 -0.0002>, rotation=<rot 0 1.231 0>}]
 dyn_hair_side_01_r [{shapeType=Box, extentX=7.5, extentY=7.5, extentZ=2.2217, center=<3.6183 -0.0054 -2.0772>, rotation=<rot 0 0 3.9798>}]
 dyn_hair_side_02_l [{shapeType=Box, extentX=7.5, extentY=7.5, extentZ=2.2217, center=<3.6183 0 -0.0002>}]
 dyn_hair_side_02_r [{shapeType=Box, extentX=7.5, extentY=6.5, extentZ=3.7217, center=<3.6248 0.0659 -2.6244>, rotation=<rot 0 0 -6.0207>}]
 dyn_hair_side_03_l [{shapeType=Box, extentX=7.5, extentY=7.5, extentZ=2.2217, center=<3.6183 0 -0.0002>}]
 dyn_hair_side_03_r [{shapeType=Box, extentX=7.5, extentY=8.5, extentZ=4.2217, center=<3.6183 0 -2>}]
 dyn_hair_side_04_l [{shapeType=Box, extentX=7.5, extentY=7.5, extentZ=2.2217, center=<3.6183 0 -0.0002>}]
 dyn_hair_side_04_r [{shapeType=Box, extentX=9.1977, extentY=7.5, extentZ=2.2217, center=<4.2956 0.0001 -0.0001>, rotation=<rot 0 0 36.6325>}]
 dyn_hair_side_05_l [{shapeType=Box, extentX=8.5209, extentY=9, extentZ=3.3231, center=<2.7586 0 -2>}]
 dyn_hair_side_base_l [{shapeType=Sphere, radius=0.5}]
 dyn_hair_side_base_r [{shapeType=Sphere, radius=0.5}]
 dyn_hair_side_parent_l [{shapeType=Box, extentX=7.5, extentY=7.5, extentZ=2.2217, center=<3.6183 0 -0.0002>}]
 dyn_hair_side_parent_r [{shapeType=Box, extentX=7.5, extentY=7.5, extentZ=2.2217, center=<3.6183 0 -0.0002>, rotation=<rot 0 0 12.533>}]
 head [{shapeType=Capsule, length=3.625, radius=6, center=<2.4384 2.1993 -0.0001>, rotation=<rot 90 0 -9.2448>}]
 neck_02 [{shapeType=Capsule, length=4.2601, radius=6.7399, center=<-2.6691 0.5785 0>, rotation=<rot 90 0 0>}]
 spine_05 [
  {shapeType=Capsule, length=9, radius=13, center=<-2.8024 0.7906 -0.0001>, rotation=<rot 90 0 -10.2932>}
  {shapeType=Capsule, length=15.9985, radius=6.7515, center=<6.4419 -0.8522 -5.8821>}
  {shapeType=Capsule, length=15.2485, radius=7.5015, center=<2.6458 -0.8521 3.8862>}
 ]
 upperarm_l [{shapeType=Capsule, length=24.6067, radius=6.0234, center=<11.5868 0.0001 0>, rotation=<rot 90 0 0>}]
 upperarm_r [{shapeType=Capsule, length=18.629, radius=5.2494, center=<-10.7816 -0.0001 0>, rotation=<rot 90 0 0>}]
}
constraints [
 {bone1Name=dyn_hair_front_parent_r, bone2Name=dyn_hair_front_base_r, swing1LimitDegrees=35, swing2LimitDegrees=15, twistLimitDegrees=5}
 {bone1Name=dyn_hair_front_01_r, bone2Name=dyn_hair_front_parent_r, swing1LimitDegrees=30, swing2LimitDegrees=15, twistLimitDegrees=5}
 {bone1Name=dyn_hair_front_02_r, bone2Name=dyn_hair_front_01_r, swing1LimitDegrees=30, swing2LimitDegrees=15, twistLimitDegrees=5}
 {bone1Name=dyn_hair_front_03_r, bone2Name=dyn_hair_front_02_r, swing1LimitDegrees=30, swing2LimitDegrees=15, twistLimitDegrees=5}
 {bone1Name=dyn_hair_side_parent_r, bone2Name=dyn_hair_side_base_r, swing1LimitDegrees=5, swing2LimitDegrees=35, twistLimitDegrees=5}
 {bone1Name=dyn_hair_side_01_r, bone2Name=dyn_hair_side_parent_r, swing1LimitDegrees=5, swing2LimitDegrees=40, twistLimitDegrees=5}
 {bone1Name=dyn_hair_side_02_r, bone2Name=dyn_hair_side_01_r, swing1LimitDegrees=5, swing2LimitDegrees=30, twistLimitDegrees=5}
 {bone1Name=dyn_hair_side_03_r, bone2Name=dyn_hair_side_02_r, swing1LimitDegrees=5, swing2LimitDegrees=30, twistLimitDegrees=5}
 {bone1Name=dyn_hair_side_04_r, bone2Name=dyn_hair_side_03_r, swing1LimitDegrees=4, swing2LimitDegrees=30, twistLimitDegrees=5}
 {bone1Name=dyn_hair_back_parent_mid, bone2Name=dyn_hair_back_base_mid, swing1LimitDegrees=5, swing2LimitDegrees=35, twistLimitDegrees=5}
 {bone1Name=dyn_hair_back_01_mid, bone2Name=dyn_hair_back_parent_mid, swing1LimitDegrees=5, swing2LimitDegrees=40, twistLimitDegrees=5}
 {bone1Name=dyn_hair_back_02_mid, bone2Name=dyn_hair_back_01_mid, swing1LimitDegrees=5, swing2LimitDegrees=25, twistLimitDegrees=5}
 {bone1Name=dyn_hair_back_03_mid, bone2Name=dyn_hair_back_02_mid, swing1LimitDegrees=5, swing2LimitDegrees=20, twistLimitDegrees=5}
 {bone1Name=dyn_hair_back_04_mid, bone2Name=dyn_hair_back_03_mid, swing1LimitDegrees=5, swing2LimitDegrees=20, twistLimitDegrees=5}
 {bone1Name=dyn_hair_back_05_mid, bone2Name=dyn_hair_back_04_mid, swing1LimitDegrees=5, swing2LimitDegrees=20, twistLimitDegrees=5}
 {bone1Name=dyn_hair_side_parent_l, bone2Name=dyn_hair_side_base_l, swing1LimitDegrees=5, swing2LimitDegrees=35, twistLimitDegrees=5}
 {bone1Name=dyn_hair_side_01_l, bone2Name=dyn_hair_side_parent_l, swing1LimitDegrees=5, swing2LimitDegrees=40, twistLimitDegrees=5}
 {bone1Name=dyn_hair_side_02_l, bone2Name=dyn_hair_side_01_l, swing1LimitDegrees=5, swing2LimitDegrees=30, twistLimitDegrees=5}
 {bone1Name=dyn_hair_side_03_l, bone2Name=dyn_hair_side_02_l, swing1LimitDegrees=5, swing2LimitDegrees=20, twistLimitDegrees=5}
 {bone1Name=dyn_hair_side_04_l, bone2Name=dyn_hair_side_03_l, swing1LimitDegrees=5, swing2LimitDegrees=20, twistLimitDegrees=5}
 {bone1Name=dyn_hair_side_05_l, bone2Name=dyn_hair_side_04_l, swing1LimitDegrees=5, swing2LimitDegrees=20, twistLimitDegrees=5}
 {bone1Name=dyn_hair_front_parent_l, bone2Name=dyn_hair_front_base_l, swing1LimitDegrees=5, swing2LimitDegrees=30, twistLimitDegrees=5}
 {bone1Name=dyn_hair_front_01_l, bone2Name=dyn_hair_front_parent_l, swing1LimitDegrees=5, swing2LimitDegrees=30, twistLimitDegrees=5}
 {bone1Name=dyn_hair_front_02_l, bone2Name=dyn_hair_front_01_l, swing1LimitDegrees=5, swing2LimitDegrees=20, twistLimitDegrees=8}
 {bone1Name=dyn_hair_front_03_l, bone2Name=dyn_hair_front_02_l, swing1LimitDegrees=5, swing2LimitDegrees=30, twistLimitDegrees=8}
 {bone1Name=dyn_hair_front_04_l, bone2Name=dyn_hair_front_03_l, swing1LimitDegrees=5, swing2LimitDegrees=30, twistLimitDegrees=8}
 {bone1Name=dyn_bangs_parent_r, bone2Name=dyn_bangs_base_r, swing1LimitDegrees=10, swing2LimitDegrees=4, twistLimitDegrees=3}
 {bone1Name=dyn_bangs_01_r, bone2Name=dyn_bangs_parent_r, swing1LimitDegrees=10, swing2LimitDegrees=8, twistLimitDegrees=3}
 {bone1Name=dyn_bangs_parent_mid, bone2Name=dyn_bangs_base_mid, swing1LimitDegrees=10, swing2LimitDegrees=4, twistLimitDegrees=3}
 {bone1Name=dyn_bangs_01_mid, bone2Name=dyn_bangs_parent_mid, swing1LimitDegrees=10, swing2LimitDegrees=8, twistLimitDegrees=3}
 {bone1Name=dyn_bangs_parent_l, bone2Name=dyn_bangs_base_l, swing1LimitDegrees=10, swing2LimitDegrees=4, twistLimitDegrees=3}
 {bone1Name=dyn_bangs_01_l, bone2Name=dyn_bangs_parent_l, swing1LimitDegrees=10, swing2LimitDegrees=8, twistLimitDegrees=3}
 {bone1Name=dyn_hair_side_04_l, bone2Name=dyn_hair_back_04_mid, swing1LimitDegrees=45, swing2LimitDegrees=45, twistLimitDegrees=45}
 {
  bone1Name=dyn_hair_side_03_r
  bone2Name=dyn_hair_back_03_mid
  swing1LimitDegrees=45
  swing1Motion=Free
  swing2LimitDegrees=45
  swing2Motion=Free
  twistLimitDegrees=45
  twistMotion=Free
 }
 {bone1Name=dyn_hair_front_03_l, bone2Name=dyn_hair_side_03_l, swing1LimitDegrees=45, swing2LimitDegrees=45, twistLimitDegrees=45}
 {bone1Name=dyn_hair_front_04_l, bone2Name=dyn_hair_side_04_l, swing1LimitDegrees=45, swing2LimitDegrees=45, twistLimitDegrees=45}
 {
  bone1Name=dyn_hair_side_04_r
  bone2Name=dyn_hair_back_05_mid
  swing1LimitDegrees=45
  swing1Motion=Free
  swing2LimitDegrees=45
  swing2Motion=Free
  twistLimitDegrees=45
  twistMotion=Free
 }
 {
  bone1Name=dyn_hair_front_02_l
  bone2Name=dyn_hair_side_02_l
  swing1LimitDegrees=45
  swing1Motion=Free
  swing2LimitDegrees=45
  swing2Motion=Free
  twistLimitDegrees=45
  twistMotion=Free
 }
 {
  bone1Name=dyn_hair_side_02_l
  bone2Name=dyn_hair_back_02_mid
  swing1LimitDegrees=45
  swing1Motion=Free
  swing2LimitDegrees=45
  swing2Motion=Free
  twistLimitDegrees=45
  twistMotion=Free
 }
 {bone1Name=dyn_hair_side_03_l, bone2Name=dyn_hair_back_03_mid, swing1LimitDegrees=45, swing2LimitDegrees=45, twistLimitDegrees=45}
 {bone1Name=dyn_hair_side_05_l, bone2Name=dyn_hair_back_05_mid, swing1LimitDegrees=45, swing2LimitDegrees=45, twistLimitDegrees=45}
]
