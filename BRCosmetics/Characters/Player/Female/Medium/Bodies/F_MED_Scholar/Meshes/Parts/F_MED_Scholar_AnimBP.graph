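# Post-process style AnimBP for the F_MED_Scholar body part, driving secondary motion on the hair and choker.
# It copies the parent mesh pose, constrains the choker, then selects RigidBody branches for skydiving without parachute, short backbling or none, and DBNO.
# Weighted constraints pin hair when downed, and additive bone offsets shape it elsewhere.
# The EventGraph derives jump impulse, state alphas and clavicle-driven melee offsets, resetting dynamics after skydiving.
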AnimBlueprint /BRCosmetics/Characters/Player/Female/Medium/Bodies/F_MED_Scholar/Meshes/Parts/F_MED_Scholar_AnimBP
target_skeleton=/Game/Characters/Player/Female/Medium/Bodies/F_MED_Scholar/Meshes/Parts/F_MED_Scholar_Skeleton
graphs [
 AnimGraph {nodes [
   n1 = Get DBNO() -> (DBNO:double)
   n2 = "Constraint - dyn_hair_side_05_l (PPPPPPPP)"!(
    ComponentPose=Local To Component(LocalPose=Use cached pose 'BackpackCache'())
    Alpha=@n1
    ConstraintWeights_0=0.0345
    ConstraintWeights_1:float=0.0026
    ConstraintWeights_2:float=0.1132
    ConstraintWeights_3:float=0.0078
    ConstraintWeights_4:float=0.3381
    ConstraintWeights_5:float=0.4865
    ConstraintWeights_6:float=0.017
    ConstraintWeights_7:float=0.0003
   ) -> (Pose) {
    alphaBoolBlend {blendOption=Linear}
    boneToModify {boneName=dyn_hair_side_05_l}
    constraintSetup [
     {targetBone {boneName=upperarm_twist_01_l}}
     {targetBone {boneName=spine_02}}
     {targetBone {boneName=spine_04}}
     {targetBone {boneName=spine_03}}
     {targetBone {boneName=spine_05}}
     {targetBone {boneName=clavicle_l}}
     {targetBone {boneName=clavicle_r}}
     {targetBone {boneName=spine_01}}
    ]
    constraintWeights [1,1,1,1,1,1,1,1]
   }
   n3 = "Constraint - dyn_hair_back_05_mid (PPPPPPP)"!(
    ComponentPose=@n2
    Alpha=@n1
    ConstraintWeights_0=0.0002
    ConstraintWeights_1:float=0.0911
    ConstraintWeights_2:float=0.0007
    ConstraintWeights_3:float=0.6069
    ConstraintWeights_4:float=0.1527
    ConstraintWeights_5:float=0.1483
    ConstraintWeights_6:float=0.0001
   ) -> (Pose) {
    alphaBoolBlend {blendOption=Linear}
    boneToModify {boneName=dyn_hair_back_05_mid}
    constraintSetup [
     {targetBone {boneName=spine_02}}
     {targetBone {boneName=spine_04}}
     {targetBone {boneName=spine_03}}
     {targetBone {boneName=spine_05}}
     {targetBone {boneName=clavicle_l}}
     {targetBone {boneName=clavicle_r}}
     {targetBone {boneName=spine_01}}
    ]
    constraintWeights [1,1,1,1,1,1,1]
   }
   n4 = "Constraint - dyn_hair_side_04_r (PPPPPPP)"!(
    ComponentPose=@n3
    Alpha=@n1
    ConstraintWeights_0=0.0001
    ConstraintWeights_1:float=0.0489
    ConstraintWeights_2:float=0.0005
    ConstraintWeights_3:float=0.2458
    ConstraintWeights_4:float=0.0027
    ConstraintWeights_5:float=0.6253
    ConstraintWeights_6:float=0.0767
   ) -> (Pose) {
    alphaBoolBlend {blendOption=Linear}
    boneToModify {boneName=dyn_hair_side_04_r}
    constraintSetup [
     {targetBone {boneName=spine_02}}
     {targetBone {boneName=spine_04}}
     {targetBone {boneName=spine_03}}
     {targetBone {boneName=spine_05}}
     {targetBone {boneName=clavicle_l}}
     {targetBone {boneName=clavicle_r}}
     {targetBone {boneName=upperarm_twist_01_r}}
    ]
    constraintWeights [1,1,1,1,1,1,1]
   }
   n5 = "Constraint - dyn_hair_side_04_l (PPPPP)"!(
    ComponentPose=@n4
    Alpha=@n1
    ConstraintWeights_0=0.0207
    ConstraintWeights_1:float=0.0082
    ConstraintWeights_2:float=0.3946
    ConstraintWeights_3:float=0.5644
    ConstraintWeights_4:float=0.0121
   ) -> (Pose) {
    alphaBoolBlend {blendOption=Linear}
    boneToModify {boneName=dyn_hair_side_04_l}
    constraintSetup [
     {targetBone {boneName=upperarm_twist_01_l}}
     {targetBone {boneName=spine_04}}
     {targetBone {boneName=spine_05}}
     {targetBone {boneName=clavicle_l}}
     {targetBone {boneName=clavicle_r}}
    ]
    constraintWeights [1,1,1,1,1]
   }
   n6 = "Constraint - dyn_hair_back_04_mid (PPP)"!(ComponentPose=@n5, Alpha=@n1, ConstraintWeights_0=0.8813, ConstraintWeights_1:float=0.0594, ConstraintWeights_2:float=0.0594) -> (Pose) {
    alphaBoolBlend {blendOption=Linear}
    boneToModify {boneName=dyn_hair_back_04_mid}
    constraintSetup [{targetBone {boneName=spine_05}}, {targetBone {boneName=clavicle_l}}, {targetBone {boneName=clavicle_r}}]
    constraintWeights [1,1,1]
   }
   n7 = RigidBody(ComponentPose=@n6, Alpha=Get DBNO_Carried() -> (DBNO_Carried:double)) {
    bClampLinearTranslationLimitToRefPose=true
    bOverrideWorldGravity=true
    overridePhysicsAsset=&/F_MED_Scholar_DBNO_Physics
    simulationSpace=BaseBoneSpace
    alphaBoolBlend {blendOption=Linear}
    alphaScaleBiasClamp {bInterpResult=true, interpSpeedDecreasing=4, interpSpeedIncreasing=6}
    baseBoneRef {boneName=spine_05}
    componentAppliedLinearAccClamp=<3000 3000 3000>
    componentLinearAccScale=<0.1 0.1 0.2>
    componentLinearVelScale=<0.1 0.1 0.2>
    overrideWorldGravity {z=-875}
   }
   n8 = "Transform (Modify) Bone - Bone: dyn_hair_side_base_r"!(ComponentPose=Local To Component(LocalPose=Use cached pose 'BackpackCache'()), Alpha=0.5, Rotation="-20,0,0") -> (Pose) {
    rotationMode=BMM_Additive
    rotationSpace=BCS_BoneSpace
    alphaBoolBlend {blendOption=Linear}
    alphaScaleBiasClamp {bInterpResult=true, interpSpeedDecreasing=5, interpSpeedIncreasing=5}
    boneToModify {boneName=dyn_hair_side_base_r}
    rotation {pitch=-20}
   }
   n9 = "Transform (Modify) Bone - Bone: dyn_hair_back_base_mid"!(ComponentPose=@n8, Alpha=0.5, Rotation="-45.000,-0.217,0.113") -> (Pose) {
    rotationMode=BMM_Additive
    rotationSpace=BCS_BoneSpace
    alphaBoolBlend {blendOption=Linear}
    alphaScaleBiasClamp {bInterpResult=true, interpSpeedDecreasing=5, interpSpeedIncreasing=5}
    boneToModify {boneName=dyn_hair_back_base_mid}
    rotation=<rot -45 -0.217 0.113>
   }
   n10 = RigidBody(ComponentPose=@n9) {
    bClampLinearTranslationLimitToRefPose=true
    bOverrideWorldGravity=true
    overridePhysicsAsset=/Game/Characters/Player/Female/Medium/Bodies/F_MED_Scholar/Meshes/Parts/F_MED_Scholar_Backpack_Physics
    simulationSpace=BaseBoneSpace
    alphaBoolBlend {blendOption=Linear}
    alphaScaleBiasClamp {bInterpResult=true, interpSpeedDecreasing=4, interpSpeedIncreasing=6}
    baseBoneRef {boneName=spine_05}
    componentAppliedLinearAccClamp=<3000 3000 3000>
    componentLinearAccScale=<0.1 0.1 0.2>
    componentLinearVelScale=<0.1 0.1 0.2>
    overrideWorldGravity {z=-875}
   }
   n11 = Blend Poses by bool(BlendPose_0=Component To Local(ComponentPose=@n7), BlendPose_1=Component To Local(ComponentPose=@n10), bActiveValue=Get bIsDBNO()) {blendType=Linear}
   n12 = Get clavicle_L_pickaxe() -> (clavicle_L_pickaxe:double)
   n13 = "Transform (Modify) Bone - Bone: dyn_hair_front_base_l"(
    ComponentPose=Local To Component(LocalPose=Use cached pose 'BackpackCache'())
    Alpha=@n12
    Translation="0,0,0"
    Rotation="-90,0,0"
    Scale="1,1,1"
   ) {
    rotationMode=BMM_Additive
    rotationSpace=BCS_BoneSpace
    scaleMode=BMM_Additive
    scaleSpace=BCS_BoneSpace
    translationMode=BMM_Additive
    translationSpace=BCS_BoneSpace
    alphaBoolBlend {blendOption=Linear}
    alphaScaleBiasClamp {bInterpResult=true, interpSpeedDecreasing=15, interpSpeedIncreasing=15}
    boneToModify {boneName=dyn_hair_front_base_l}
    rotation {pitch=-90}
   }
   n14 = "Transform (Modify) Bone - Bone: dyn_hair_side_base_l"(ComponentPose=@n13, Alpha=@n12, Translation="0,0,0", Rotation="-90,0,0", Scale="1,1,1") {
    rotationMode=BMM_Additive
    rotationSpace=BCS_BoneSpace
    scaleMode=BMM_Additive
    scaleSpace=BCS_BoneSpace
    translationMode=BMM_Additive
    translationSpace=BCS_BoneSpace
    alphaBoolBlend {blendOption=Linear}
    alphaScaleBiasClamp {bInterpResult=true, interpSpeedDecreasing=15, interpSpeedIncreasing=15}
    boneToModify {boneName=dyn_hair_side_base_l}
    rotation {pitch=-90}
   }
   n15 = Get clavicle_R_pickaxe() -> (clavicle_R_pickaxe:double)
   n16 = "Transform (Modify) Bone - Bone: dyn_hair_back_base_mid"!(ComponentPose=@n14, Alpha=@n15, Translation:Vector="0,0,0", Rotation="-90,0,0", Scale:Vector="1,1,1") -> (Pose) {
    rotationMode=BMM_Additive
    rotationSpace=BCS_BoneSpace
    scaleMode=BMM_Additive
    scaleSpace=BCS_BoneSpace
    translationMode=BMM_Additive
    translationSpace=BCS_BoneSpace
    alphaBoolBlend {blendOption=Linear}
    alphaScaleBiasClamp {bInterpResult=true, interpSpeedDecreasing=15, interpSpeedIncreasing=15}
    boneToModify {boneName=dyn_hair_back_base_mid}
    rotation {pitch=-90}
   }
   n17 = "Transform (Modify) Bone - Bone: dyn_hair_side_base_r"!(ComponentPose=@n16, Alpha=@n15, Translation:Vector="0,0,0", Rotation="-90,0,0", Scale:Vector="1,1,1") -> (Pose) {
    rotationMode=BMM_Replace
    rotationSpace=BCS_BoneSpace
    scaleMode=BMM_Additive
    scaleSpace=BCS_BoneSpace
    translationMode=BMM_Additive
    translationSpace=BCS_BoneSpace
    alphaBoolBlend {blendOption=Linear}
    alphaScaleBiasClamp {bInterpResult=true, interpSpeedDecreasing=15, interpSpeedIncreasing=15}
    boneToModify {boneName=dyn_hair_side_base_r}
    rotation {pitch=-90}
   }
   n18 = "Transform (Modify) Bone - Bone: dyn_hair_front_base_r"(ComponentPose=@n17, Alpha=@n15, Translation="0,0,0", Rotation="-90,0,0", Scale="1,1,1") {
    rotationMode=BMM_Additive
    rotationSpace=BCS_BoneSpace
    scaleMode=BMM_Additive
    scaleSpace=BCS_BoneSpace
    translationMode=BMM_Additive
    translationSpace=BCS_BoneSpace
    alphaBoolBlend {blendOption=Linear}
    alphaScaleBiasClamp {bInterpResult=true, interpSpeedDecreasing=15, interpSpeedIncreasing=15}
    boneToModify {boneName=dyn_hair_front_base_r}
    rotation {pitch=-90}
   }
   n19 = "Transform (Modify) Bone - Bone: dyn_hair_front_02_r"(ComponentPose=@n18, Alpha=@n15, Translation="0,0,20", Rotation="-90,0,0", Scale="1,1,1") {
    rotationMode=BMM_Additive
    rotationSpace=BCS_BoneSpace
    scaleMode=BMM_Additive
    scaleSpace=BCS_BoneSpace
    translationMode=BMM_Additive
    translationSpace=BCS_BoneSpace
    alphaBoolBlend {blendOption=Linear}
    alphaScaleBiasClamp {bInterpResult=true, interpSpeedDecreasing=15, interpSpeedIncreasing=15}
    boneToModify {boneName=dyn_hair_front_02_r}
    rotation {pitch=-90}
    translation {z=20}
   }
   n20 = RigidBody(ComponentPose=@n19, ExternalForce=Get Jump_Impulse() -> (Jump_Impulse:Vector)) {
    bClampLinearTranslationLimitToRefPose=true
    bOverrideWorldGravity=true
    overridePhysicsAsset=/Game/Characters/Player/Female/Medium/Bodies/F_MED_Scholar/Meshes/Parts/F_MED_Scholar_Physics
    simulationSpace=BaseBoneSpace
    alphaBoolBlend {blendOption=Linear}
    alphaScaleBiasClamp {bInterpResult=true, interpSpeedDecreasing=4, interpSpeedIncreasing=6}
    baseBoneRef {boneName=spine_05}
    componentAppliedLinearAccClamp=<3000 3000 3000>
    componentLinearAccScale=<0.1 0.1 0.2>
    componentLinearVelScale=<0.1 0.1 0.2>
    overrideWorldGravity {z=-980}
   }
   n21 = Get DBNO()
   n22 = "Constraint - dyn_hair_side_05_l (PPPPPPPP)"(
    ComponentPose=Local To Component(LocalPose=Use cached pose 'BackpackCache'())
    Alpha=@n21
    ConstraintWeights_0=0.0345
    ConstraintWeights_1=0.0026
    ConstraintWeights_2=0.1132
    ConstraintWeights_3=0.0078
    ConstraintWeights_4=0.3381
    ConstraintWeights_5=0.4865
    ConstraintWeights_6=0.017
    ConstraintWeights_7=0.0003
   ) {
    alphaBoolBlend {blendOption=Linear}
    boneToModify {boneName=dyn_hair_side_05_l}
    constraintSetup [
     {targetBone {boneName=upperarm_twist_01_l}}
     {targetBone {boneName=spine_02}}
     {targetBone {boneName=spine_04}}
     {targetBone {boneName=spine_03}}
     {targetBone {boneName=spine_05}}
     {targetBone {boneName=clavicle_l}}
     {targetBone {boneName=clavicle_r}}
     {targetBone {boneName=spine_01}}
    ]
    constraintWeights [1,1,1,1,1,1,1,1]
   }
   n23 = "Constraint - dyn_hair_back_05_mid (PPPPPPP)"(
    ComponentPose=@n22
    Alpha=@n21
    ConstraintWeights_0=0.0002
    ConstraintWeights_1=0.0911
    ConstraintWeights_2=0.0007
    ConstraintWeights_3=0.6069
    ConstraintWeights_4=0.1527
    ConstraintWeights_5=0.1483
    ConstraintWeights_6=0.0001
   ) {
    alphaBoolBlend {blendOption=Linear}
    boneToModify {boneName=dyn_hair_back_05_mid}
    constraintSetup [
     {targetBone {boneName=spine_02}}
     {targetBone {boneName=spine_04}}
     {targetBone {boneName=spine_03}}
     {targetBone {boneName=spine_05}}
     {targetBone {boneName=clavicle_l}}
     {targetBone {boneName=clavicle_r}}
     {targetBone {boneName=spine_01}}
    ]
    constraintWeights [1,1,1,1,1,1,1]
   }
   n24 = "Constraint - dyn_hair_side_04_r (PPPPPPP)"(
    ComponentPose=@n23
    Alpha=@n21
    ConstraintWeights_0=0.0001
    ConstraintWeights_1=0.0489
    ConstraintWeights_2=0.0005
    ConstraintWeights_3=0.2458
    ConstraintWeights_4=0.0027
    ConstraintWeights_5=0.6253
    ConstraintWeights_6=0.0767
   ) {
    alphaBoolBlend {blendOption=Linear}
    boneToModify {boneName=dyn_hair_side_04_r}
    constraintSetup [
     {targetBone {boneName=spine_02}}
     {targetBone {boneName=spine_04}}
     {targetBone {boneName=spine_03}}
     {targetBone {boneName=spine_05}}
     {targetBone {boneName=clavicle_l}}
     {targetBone {boneName=clavicle_r}}
     {targetBone {boneName=upperarm_twist_01_r}}
    ]
    constraintWeights [1,1,1,1,1,1,1]
   }
   n25 = "Constraint - dyn_hair_side_04_l (PPPPP)"(
    ComponentPose=@n24
    Alpha=@n21
    ConstraintWeights_0=0.0207
    ConstraintWeights_1=0.0082
    ConstraintWeights_2=0.3946
    ConstraintWeights_3=0.5644
    ConstraintWeights_4=0.0121
   ) {
    alphaBoolBlend {blendOption=Linear}
    boneToModify {boneName=dyn_hair_side_04_l}
    constraintSetup [
     {targetBone {boneName=upperarm_twist_01_l}}
     {targetBone {boneName=spine_04}}
     {targetBone {boneName=spine_05}}
     {targetBone {boneName=clavicle_l}}
     {targetBone {boneName=clavicle_r}}
    ]
    constraintWeights [1,1,1,1,1]
   }
   n26 = "Constraint - dyn_hair_back_04_mid (PPP)"(ComponentPose=@n25, Alpha=@n21, ConstraintWeights_0=0.8813, ConstraintWeights_1=0.0594, ConstraintWeights_2=0.0594) {
    alphaBoolBlend {blendOption=Linear}
    boneToModify {boneName=dyn_hair_back_04_mid}
    constraintSetup [{targetBone {boneName=spine_05}}, {targetBone {boneName=clavicle_l}}, {targetBone {boneName=clavicle_r}}]
    constraintWeights [1,1,1]
   }
   n27 = RigidBody(ComponentPose=@n26, Alpha=Get DBNO_Carried(), ExternalForce=Get Jump_Impulse()) {
    bClampLinearTranslationLimitToRefPose=true
    bOverrideWorldGravity=true
    overridePhysicsAsset=&F_MED_Scholar_DBNO_Physics
    simulationSpace=BaseBoneSpace
    alphaBoolBlend {blendOption=Linear}
    alphaScaleBiasClamp {bInterpResult=true, interpSpeedDecreasing=4, interpSpeedIncreasing=6}
    baseBoneRef {boneName=spine_05}
    componentAppliedLinearAccClamp=<3000 3000 3000>
    componentLinearAccScale=<0.1 0.1 0.2>
    componentLinearVelScale=<0.1 0.1 0.2>
    overrideWorldGravity {z=-980}
   }
   n28 = Blend Poses by bool(BlendPose_0=Component To Local(ComponentPose=@n27), BlendPose_1=Component To Local(ComponentPose=@n20), bActiveValue=Get bIsDBNO()) {blendType=Linear}
   n29 = Get isWearingPackBool() -> (isWearingPackBool:bool)
   n30 = "Transform (Modify) Bone - Bone: dyn_hair_front_parent_l"!(
    ComponentPose=Local To Component(LocalPose=Use cached pose 'BackpackCache'())
    Alpha=Get SkyFloat_Alpha() -> (SkyFloat_Alpha:double)
    Rotation="-50,0.000,0.000"
   ) -> (Pose) {
    rotationMode=BMM_Additive
    rotationSpace=BCS_BoneSpace
    alphaBoolBlend {blendOption=Linear}
    alphaScaleBiasClamp {bInterpResult=true, interpSpeedDecreasing=5, interpSpeedIncreasing=5}
    boneToModify {boneName=dyn_hair_front_parent_l}
    rotation {pitch=-50}
   }
   n31 = "Transform (Modify) Bone - Bone: dyn_hair_side_parent_l"!(ComponentPose=@n30, Alpha=Get SkyFloat_Alpha(), Rotation="-40,0.000,0.000") -> (Pose) {
    rotationMode=BMM_Additive
    rotationSpace=BCS_BoneSpace
    alphaBoolBlend {blendOption=Linear}
    alphaScaleBiasClamp {bInterpResult=true, interpSpeedDecreasing=5, interpSpeedIncreasing=5}
    boneToModify {boneName=dyn_hair_side_parent_l}
    rotation {pitch=-40}
   }
   n32 = "Transform (Modify) Bone - Bone: dyn_hair_back_parent_mid"!(ComponentPose=@n31, Alpha=Get SkyFloat_Alpha(), Rotation="-40,0.000,0.000") -> (Pose) {
    rotationMode=BMM_Additive
    rotationSpace=BCS_BoneSpace
    alphaBoolBlend {blendOption=Linear}
    alphaScaleBiasClamp {bInterpResult=true, interpSpeedDecreasing=5, interpSpeedIncreasing=5}
    boneToModify {boneName=dyn_hair_back_parent_mid}
    rotation {pitch=-40}
   }
   n33 = "Transform (Modify) Bone - Bone: dyn_hair_side_parent_r"!(ComponentPose=@n32, Alpha=Get SkyFloat_Alpha(), Rotation="-40,0.000,0.000") -> (Pose) {
    rotationMode=BMM_Additive
    rotationSpace=BCS_BoneSpace
    alphaBoolBlend {blendOption=Linear}
    alphaScaleBiasClamp {bInterpResult=true, interpSpeedDecreasing=5, interpSpeedIncreasing=5}
    boneToModify {boneName=dyn_hair_side_parent_r}
    rotation {pitch=-40}
   }
   n34 = "Transform (Modify) Bone - Bone: dyn_hair_front_parent_r"!(ComponentPose=@n33, Alpha=Get SkyFloat_Alpha(), Rotation="-25.614,-64.011,32.279") -> (Pose) {
    rotationMode=BMM_Additive
    rotationSpace=BCS_BoneSpace
    alphaBoolBlend {blendOption=Linear}
    alphaScaleBiasClamp {bInterpResult=true, interpSpeedDecreasing=5, interpSpeedIncreasing=5}
    boneToModify {boneName=dyn_hair_front_parent_r}
    rotation=<rot -25.614 -64.011 32.279>
   }
   n35 = RigidBody(ComponentPose=@n34) {
    bClampLinearTranslationLimitToRefPose=true
    bOverrideWorldGravity=true
    overridePhysicsAsset=&/F_MED_Scholar_SD
    simulationSpace=BaseBoneSpace
    alphaBoolBlend {blendOption=Linear}
    alphaScaleBiasClamp {bInterpResult=true, interpSpeedDecreasing=4, interpSpeedIncreasing=6}
    baseBoneRef {boneName=spine_04}
    componentAppliedLinearAccClamp=<3000 3000 3000>
    componentLinearAccScale=<0.2 0.2 0.3>
    componentLinearVelScale=<0.2 0.2 0.3>
    overrideWorldGravity {z=-300}
   }
   Output Pose(Result=Blend Poses by bool(
     BlendPose_0=Component To Local(ComponentPose=@n35)
     BlendPose_1=Blend Poses by bool(BlendPose_0=@n11, BlendPose_1=@n28, bActiveValue=@n29) {blendType=Linear}
     BlendTime_0=0.2
     BlendTime_1=0.2
     bActiveValue=Get isSkydiveNoParachute() -> (isSkydiveNoParachute:bool)
    ) {blendType=Linear}) {layerGroup=None}
   n36 = "Constraint - const_choker (PPPPPP)"!(
    ComponentPose=Local To Component(LocalPose=Copy Pose From Mesh() {bUseAttachedParent=true})
    Alpha=1
    ConstraintWeights_0=0.1217
    ConstraintWeights_1:float=0.0042
    ConstraintWeights_2:float=0.3915
    ConstraintWeights_3:float=0.0032
    ConstraintWeights_4:float=0.3004
    ConstraintWeights_5:float=0.1789
   ) -> (Pose) {
    alphaBoolBlend {blendOption=Linear}
    boneToModify {boneName=const_choker}
    constraintSetup [
     {targetBone {boneName=head}}
     {targetBone {boneName=clavicle_l}}
     {targetBone {boneName=spine_05}}
     {targetBone {boneName=clavicle_r}}
     {targetBone {boneName=neck_02}}
     {targetBone {boneName=neck_01}}
    ]
    constraintWeights [1,1,1,1,1,1]
   }
   Save cached pose 'BackpackCache'(Pose=Component To Local(ComponentPose=@n36)) {cacheName=BackpackCache}
  ]}
 EventGraph {nodes [
   Event OnCharacterCustomizationLoaded()
   n1 = CheckForBackbling()
   > Set isWearingPackBool(isWearingPackBool:bool=AND Boolean(A=@n1, B=NOT Boolean(A=@n1.bIsBackblingLong))) -> (Output_Get:bool)
   Event BlueprintInitializeAnimation()
   > n2 = Cast To PlayerPawn_Athena(Object=TryGetPawnOwner())
   > n3 = Set Pawn Owner(Pawn Owner:Player Pawn Athena Object Reference=@n2) -> (Output_Get:Player Pawn Athena Object Reference)
   > Set Body_Mesh(Body_Mesh=GetSkeletalMeshForPartType(self=@n3))
   > Set Skel Mesh(Skel Mesh=GetOwnerSkeletalMeshComponent())
   Event BlueprintUpdateAnimation()
   > n4 = Sequence!(execute) -> (then_0, then_1, then_2:Exec, then_3:Exec, then_4:Exec)
   > Set sky_Alpha(sky_Alpha:double=SelectFloat(A=0.5, B=1, bPickA=Get bIsSkydiveFloating())) -> (Output_Get:double)
   > Set SkyFloat_Alpha(SkyFloat_Alpha:double=SelectFloat(A=1, B=0.5, bPickA=Get bIsSkydiveFloating())) -> (Output_Get:double)
   n5 = Get bIsParachuteOpened()
   > Set isSkydiveNoParachute(isSkydiveNoParachute:bool=AND Boolean(A=Get bIsSkyDiving(), B=NOT Boolean(A=@n5))) -> (Output_Get:bool)
   n6 = NOT Boolean(A=OR Boolean!(A=Get bIsSkydiveDiveMode(), B=Get bIsSkyDiving(), C:bool=Get bIsSkydiveFloating(), D:bool) -> (ReturnValue))
   n7 = OR Boolean!(A=Get bIsJumping(), B=Get bIsFalling()) -> (ReturnValue)
   Set Jump_Impulse(execute=@n4.then_1, Jump_Impulse:Vector=SelectVector(A="0, 0,7000", B="0, 0, 0", bPickA=AND Boolean(A=@n6, B=@n7))) -> (Output_Get:Vector)
   Set DBNO_Carried(execute=@n4.then_2, DBNO_Carried:double=SelectFloat(A=1, B=0, bPickA=Get bIsDBNOCarried())) -> (Output_Get:double)
   Set DBNO(execute=@n4.then_3, DBNO:double=SelectFloat(A=1, B=0, bPickA=Get bIsDBNO())) -> (Output_Get:double)
   > Set IsPlayMeleeFloat(IsPlayMeleeFloat:double="To Float (Boolean)"(InBool=Get bIsPlayingMeleeAnim())) -> (Output_Get:double)
   n8 = Set clavicle_R(
    execute=@n4.then_4
    clavicle_R:double=GetAlphaFromReferenceTransformDelta(
     SkeletalMeshComponent=Get Skel Mesh()
     BoneName=clavicle_r
     BaseName=spine_05
     DeltaComponentType=EulerY
     InputRange=(X=0,Y=20)
     TwistAxis="0,0,1"
    )
   ) -> (Output_Get:double)
   > n9 = Set clavicle_L(clavicle_L:double=GetAlphaFromReferenceTransformDelta(
     SkeletalMeshComponent=Get Skel Mesh()
     BoneName=clavicle_l
     BaseName=spine_05
     DeltaComponentType=EulerY
     InputRange=(X=0,Y=20)
     TwistAxis="0,0,1"
    )) -> (Output_Get:double)
   n10 = Get Skel Mesh()
   > n11 = Branch(Condition="float >= float"(A=GetAlphaFromReferenceTransformDelta(SkeletalMeshComponent=@n10, BoneName=spine_05, BaseName=head, InputRange=(X=0,Y=60), TwistAxis="0,0,1").RawOutput))
   > Set clavicle_R_pickaxe(clavicle_R_pickaxe:double=float * float(A=Get IsPlayMeleeFloat() -> (IsPlayMeleeFloat:double), B=@n8) {is=title K2Node_PromotableOperator}) -> (Output_Get:double)
   > Set clavicle_L_pickaxe(clavicle_L_pickaxe:double=float * float(A=@n9, B=Get IsPlayMeleeFloat())) -> (Output_Get:double)
   Set clavicle_R_pickaxe(execute=@n11.else, clavicle_R_pickaxe=0.3)
   > Set clavicle_L_pickaxe(clavicle_L_pickaxe=0.3)
   Event OnEndSkydiving()
   > ResetDynamics(InTeleportType=ResetPhysics)
   Get bIsPlayingMeleeAnim()
  ]}
]
variables [
 Body_Mesh {type=SkeletalMeshComponent}
 clavicle_L {type=number}
 clavicle_L_pickaxe {type=number}
 clavicle_R {type=number}
 clavicle_R_pickaxe {type=number}
 DBNO {type=number}
 DBNO_Carried {type=number}
 Gravity Override Params {type=GravityOverrideParamsStruct}
 GravityOverride {type=Vector}
 HeadRotate_Alpha {type=number}
 isGenericBackpackAlpha {type=number}
 IsPlayMeleeFloat {type=number}
 isSkydiveNoParachute {type=boolean}
 isWearingPackBool {type=boolean}
 Jump_Impulse {type=Vector}
 Pawn Owner {type=PlayerPawn_Athena_C}
 ScholarPackAlpha {type=number}
 Skel Mesh {type=SkeletalMeshComponent}
 Skel_Mesh {type=SkeletalMeshComponent}
 sky_Alpha {type=number}
 Skydive_Alpha {type=Vector}
 SkyFloat_Alpha {type=number}
]
